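# Physics asset for the strap on the SlyJudge medium male player body, defining capsule bodies for its dynamic bone chain.
# The chain runs from spine_01 through base, parent, 01 and 02 bones, with two overlapping capsules on the last.
# The base-to-spine joint is free in swing and twist, while later joints have narrower swing and zero-degree twist limits.
# Larger capsules cover spine_01 and the left thigh helper, which has no constraint.
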PhysicsAsset /BRCosmetics/Characters/Player/Male/Medium/Bodies/M_MED_SlyJudge/Meshes/M_MED_SlyJudge_Strap_Physics
bodies {
 dyn_strap_01 [{
   shapeType=Capsule
   length=6.8793
   radius=1.6722
   shapeName=dyn_strap_01_capsule
   center=<4.0261 0 -0.7042>
   rotation=<rot 85.5902 179.9994 -179.9998>
  }]
 dyn_strap_02 [
  {
   shapeType=Capsule
   length=11.3091
   radius=2.1896
   shapeName=dyn_strap_02_capsule
   center=<8.0284 -0.0489 -0.6741>
   rotation=<rot 86.0339 131.8336 -137.5252>
  }
  {
   shapeType=Capsule
   length=12.5594
   radius=1.9475
   shapeName=dyn_strap_02_capsule
   center=<3.8943 -0.0517 -0.8517>
   rotation=<rot 85.0714 122.6829 32.9344>
  }
 ]
 dyn_strap_base [{shapeType=Capsule, length=0.505, radius=0.505, shapeName=dyn_strap_base_capsule, rotation=<rot 0 -90 0>}]
 dyn_strap_parent [{
   shapeType=Capsule
   length=4.576
   radius=1.8246
   shapeName=dyn_strap_parent_capsule
   center=<3.5329 0 -1.1619>
   rotation=<rot 86.3784 -179.9999 -179.9999>
  }]
 spine_01 [{shapeType=Capsule, length=11.5, radius=9.5, center=<-0.0502 -0.9352 0>}]
 thigh_helper_l [{shapeType=Capsule, length=25.5349, radius=9.2461, center=<-10.0406 -0.052 0.0057>, rotation=<rot 90 89.6476 -84.0787>}]
}
constraints [
 {
  bone1Name=dyn_strap_base
  bone2Name=spine_01
  swing1LimitDegrees=45
  swing1Motion=Free
  swing2LimitDegrees=45
  swing2Motion=Free
  twistLimitDegrees=45
  twistMotion=Free
 }
 {bone1Name=dyn_strap_parent, bone2Name=dyn_strap_base, swing1LimitDegrees=15, swing2LimitDegrees=45, twistLimitDegrees=0}
 {bone1Name=dyn_strap_01, bone2Name=dyn_strap_parent, swing1LimitDegrees=15, swing2LimitDegrees=30, twistLimitDegrees=0}
 {bone1Name=dyn_strap_02, bone2Name=dyn_strap_01, swing1LimitDegrees=15, swing2LimitDegrees=30, twistLimitDegrees=0}
]
