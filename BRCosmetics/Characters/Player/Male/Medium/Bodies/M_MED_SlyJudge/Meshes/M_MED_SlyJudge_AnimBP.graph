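# Post-process style AnimBlueprint for the M_MED_SlyJudge cosmetic body that adds secondary motion to its strap, shirt back, zipper and shoe tags.
# It copies the attached parent's pose, runs a Control Rig, then applies bool-blended additive bone offsets to strap, shirt and thigh helper.
# AnimDynamics and a thigh-based strap RigidBody follow, before a final strap rotation.
# Each update derives gravity, sim-space settings and state flags from movement and front-end state.
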AnimBlueprint /BRCosmetics/Characters/Player/Male/Medium/Bodies/M_MED_SlyJudge/Meshes/M_MED_SlyJudge_AnimBP
target_skeleton=&/M_MED_SlyJudge_Skeleton
graphs [
 AnimGraph {nodes [
   Output Pose(Result=Use cached pose 'AnimDynCache'()) {layerGroup=None}
   Save cached pose 'IniPoseCache'(Pose=Copy Pose From Mesh() {bUseAttachedParent=true}) {cacheName=IniPoseCache}
   n1 = "Transform (Modify) Bone - Bone: dyn_strap_base"!(
    ComponentPose=Local To Component(LocalPose=Use cached pose 'controlRigCache'())
    bAlphaBoolEnabled:bool=True
    Translation="5.386,5.930,-3.539"
    Rotation="24.642,-52.058,-28.698"
   ) -> (Pose) {
    alphaInputType=Bool
    rotationMode=BMM_Additive
    rotationSpace=BCS_BoneSpace
    translationMode=BMM_Additive
    translationSpace=BCS_BoneSpace
    alphaBoolBlend {blendInTime=0.2, blendOption=Linear, blendOutTime=0.2}
    boneToModify {boneName=dyn_strap_base}
    rotation=<rot 24.642 -52.058 -28.698>
    translation=<5.386 5.93 -3.539>
   }
   n2 = "Transform (Modify) Bone - Bone: dyn_shirt_back_base"!(ComponentPose=@n1, bAlphaBoolEnabled:bool=True, Rotation="-20,0,0") -> (Pose) {
    alphaInputType=Bool
    rotationMode=BMM_Additive
    rotationSpace=BCS_BoneSpace
    translationSpace=BCS_BoneSpace
    alphaBoolBlend {blendInTime=1, blendOption=Linear, blendOutTime=1}
    boneToModify {boneName=dyn_shirt_back_base}
    rotation {pitch=-20}
   }
   n3 = "Transform (Modify) Bone - Bone: thigh_helper_l"!(ComponentPose=@n2, bAlphaBoolEnabled:bool=True, Translation="0,-1.5,0") -> (Pose) {
    alphaInputType=Bool
    rotationMode=BMM_Additive
    rotationSpace=BCS_BoneSpace
    scaleMode=BMM_Additive
    scaleSpace=BCS_BoneSpace
    translationMode=BMM_Additive
    translationSpace=BCS_BoneSpace
    alphaBoolBlend {blendInTime=0.5, blendOption=Linear, blendOutTime=0.5}
    boneToModify {boneName=thigh_helper_l}
    translation {y=-1.5}
   }
   n4 = "Transform (Modify) Bone - Bone: dyn_strap_base"(ComponentPose=@n3, bAlphaBoolEnabled=True, Translation="2.433,-3.396,5.960", Rotation="-23.626,88.697,-61.671") {
    alphaInputType=Bool
    rotationMode=BMM_Additive
    rotationSpace=BCS_BoneSpace
    translationMode=BMM_Additive
    translationSpace=BCS_BoneSpace
    alphaBoolBlend {blendInTime=0.5, blendOption=Linear, blendOutTime=0.5}
    boneToModify {boneName=dyn_strap_base}
    rotation=<rot -23.626 88.697 -61.671>
    translation=<2.433 -3.396 5.96>
   }
   n5 = "Transform (Modify) Bone - Bone: dyn_strap_base"(ComponentPose=@n4, bAlphaBoolEnabled=True, Translation="-0.825,-0.377,4.492", Rotation="-31.577,19.481,-2.162") {
    alphaInputType=Bool
    rotationMode=BMM_Additive
    rotationSpace=BCS_BoneSpace
    translationMode=BMM_Additive
    translationSpace=BCS_BoneSpace
    alphaBoolBlend {blendInTime=0.5, blendOption=Linear, blendOutTime=0.5}
    boneToModify {boneName=dyn_strap_base}
    rotation=<rot -31.577 19.481 -2.162>
    translation=<-0.825 -0.377 4.492>
   }
   n6 = "Transform (Modify) Bone - Bone: thigh_helper_l"!(ComponentPose=@n5, bAlphaBoolEnabled=True, Translation="0,0.4,0", Rotation:Rotator="0,0,0") -> (Pose) {
    alphaInputType=Bool
    rotationMode=BMM_Additive
    rotationSpace=BCS_BoneSpace
    translationMode=BMM_Additive
    translationSpace=BCS_BoneSpace
    alphaBoolBlend {blendInTime=0.5, blendOption=Linear, blendOutTime=0.5}
    boneToModify {boneName=thigh_helper_l}
    translation {y=0.4}
   }
   n7 = "Transform (Modify) Bone - Bone: dyn_strap_base"(ComponentPose=@n6, bAlphaBoolEnabled=True, Translation="3.308,1.351,-3.004", Rotation="33.977,0.009,-0.029") {
    alphaInputType=Bool
    rotationMode=BMM_Additive
    rotationSpace=BCS_BoneSpace
    translationMode=BMM_Additive
    translationSpace=BCS_BoneSpace
    alphaBoolBlend {blendOption=Linear}
    boneToModify {boneName=dyn_strap_base}
    rotation=<rot 33.977 0.009 -0.029>
    translation=<3.308 1.351 -3.004>
   }
   n8 = "Transform (Modify) Bone - Bone: thigh_helper_l"(ComponentPose=@n7, bAlphaBoolEnabled=True, Translation="0,2,0", Rotation="0,0,0") {
    alphaInputType=Bool
    rotationMode=BMM_Additive
    rotationSpace=BCS_BoneSpace
    translationMode=BMM_Additive
    translationSpace=BCS_BoneSpace
    alphaBoolBlend {blendOption=Linear}
    boneToModify {boneName=thigh_helper_l}
    translation {y=2}
   }
   Save cached pose 'transformsModifyCache'(Pose=Component To Local(ComponentPose=@n8)) {cacheName=transformsModifyCache}
   n9 = Anim Dynamics - Bone: dyn_shoetag_l_parent!(ComponentPose=Local To Component(LocalPose=Use cached pose 'transformsModifyCache'()), Alpha=1) -> (Pose) {
    angularDampingOverride=0.9
    bGravityOverrideInSimSpace=true
    bOverrideAngularDamping=true
    bOverrideLinearDamping=true
    gravityScale=0.15
    linearDampingOverride=0.9
    simulationSpace=RootRelative
    alphaBoolBlend {blendOption=Linear}
    boundBone {boneName=dyn_shoetag_l_parent}
    componentAppliedLinearAccClamp=<6000 6000 6000>
    componentLinearAccScale=<0.15 0.15 0.15>
    componentLinearVelScale=<0.15 0.15 0.15>
    gravityOverride=<0.9 0.9 0.9>
    physicsBodyDefinitions [{
      sphereCollisionRadius=0
      boundBone {boneName=dyn_shoetag_l_parent}
      boxExtents=<20 20 20>
      constraintSetup {angularLimitsMax=<10 40 5>, angularLimitsMin=<-10 -10 -5>, angularTarget=<0 0 0>}
      localJointOffset=<3.5 0 0>
     }]
    relativeSpaceBone {boneName=calf_twist_01_r}
   }
   n10 = Anim Dynamics - Bone: dyn_shoetag_r_parent!(ComponentPose=@n9, Alpha=1) -> (Pose) {
    angularDampingOverride=0.9
    bGravityOverrideInSimSpace=true
    bOverrideAngularDamping=true
    bOverrideLinearDamping=true
    gravityScale=0.15
    linearDampingOverride=0.9
    simulationSpace=RootRelative
    alphaBoolBlend {blendOption=Linear}
    boundBone {boneName=dyn_shoetag_r_parent}
    componentAppliedLinearAccClamp=<6000 6000 6000>
    componentLinearAccScale=<0.15 0.15 0.15>
    componentLinearVelScale=<0.15 0.15 0.15>
    gravityOverride=<0.9 0.9 0.9>
    physicsBodyDefinitions [{
      sphereCollisionRadius=0
      boundBone {boneName=dyn_shoetag_r_parent}
      boxExtents=<20 20 20>
      constraintSetup {angularLimitsMax=<10 40 5>, angularLimitsMin=<-10 -10 -5>, angularTarget=<0 0 0>}
      localJointOffset=<3.5 0 0>
     }]
    relativeSpaceBone {boneName=calf_twist_01_r}
   }
   n11 = Anim Dynamics - Bone: dyn_zipper_r_parent!(ComponentPose=@n10, Alpha=1) -> (Pose) {
    angularDampingOverride=0.99
    bGravityOverrideInSimSpace=true
    bOverrideLinearDamping=true
    gravityScale=0.75
    linearDampingOverride=0.999
    linearSpringConstant=20000
    simulationSpace=RootRelative
    alphaBoolBlend {blendOption=Linear}
    boundBone {boneName=dyn_zipper_r_parent}
    componentAppliedLinearAccClamp=<6000 6000 6000>
    componentLinearAccScale=<0.25 0.25 0.15>
    componentLinearVelScale=<0.25 0.25 0.15>
    physicsBodyDefinitions [{
      sphereCollisionRadius=0
      boundBone {boneName=dyn_zipper_r_parent}
      boxExtents=<5 5 5>
      constraintSetup {angularLimitsMax=<0 5 50>, angularLimitsMin=<0 -5 -10>, angularTarget=<0 0 0>}
      localJointOffset=<2 0 0>
     }]
    relativeSpaceBone {boneName=spine_05}
   }
   n12 = Anim Dynamics - Bone: dyn_shirt_back_parent!(ComponentPose=@n11, Alpha=1) -> (Pose) {
    angularDampingOverride=1
    angularSpringConstant=20000
    bAngularSpring=true
    bOverrideAngularDamping=true
    bUseGravityOverride=true
    gravityScale=0.35
    simulationSpace=RootRelative
    alphaBoolBlend {blendOption=Linear}
    boundBone {boneName=dyn_shirt_back_parent}
    componentAppliedLinearAccClamp=<3000 3000 3000>
    componentLinearAccScale=<1 1 1>
    componentLinearVelScale=<1 1 1>
    gravityOverride {z=-980}
    physicsBodyDefinitions [{
      boundBone {boneName=dyn_shirt_back_parent}
      boxExtents=<35 35 35>
      constraintSetup {angularLimitsMax=<0 60 15>, angularLimitsMin=<0 0 -15>, angularTarget=<1 0 0>}
      localJointOffset=<8 0 0>
     }]
   }
   n13 = RigidBody!(
    ComponentPose=@n12
    Alpha
    OverrideWorldGravity:Vector="0,0,-980"
    SimSpaceSettings:Sim Space Settings Structure=(WorldAlpha=0,VelocityScaleZ=1,DampingAlpha=1,MaxLinearVelocity=10000,MaxAngularVelocity=10000,MaxLinearAcceleration=10000,MaxAngularAcceleration=10000,ExternalLinearDragV=(X=0,Y=0,Z=0),ExternalLinearVelocity=(X=0,Y=0,Z=0),ExternalAngularVelocity=(X=0,Y=0,Z=0))
   ) -> (Pose) {
    bClampLinearTranslationLimitToRefPose=true
    bOverrideWorldGravity=true
    overridePhysicsAsset=&/M_MED_SlyJudge_Strap_Physics
    simulationSpace=BaseBoneSpace
    alphaBoolBlend {blendOption=Linear}
    alphaScaleBiasClamp {bInterpResult=true, interpSpeedIncreasing=0.1}
    baseBoneRef {boneName=thigh_l}
    componentAppliedLinearAccClamp=<3000 3000 3000>
    overrideWorldGravity {z=-980}
    simSpaceSettings {maxAngularAcceleration=1, maxAngularVelocity=3, maxLinearAcceleration=800, maxLinearVelocity=300, worldAlpha=1, externalLinearDragV=<1 1 1>}
   }
   n14 = "Transform (Modify) Bone - Bone: dyn_strap_01"!(ComponentPose=@n13, bAlphaBoolEnabled:bool=True, Translation="0,0,0", Rotation="-30,0,0", Scale="1,1,1") -> (Pose) {
    alphaInputType=Bool
    rotationMode=BMM_Additive
    rotationSpace=BCS_BoneSpace
    alphaBoolBlend {blendInTime=0.5, blendOption=Linear, blendOutTime=0.5}
    boneToModify {boneName=dyn_strap_01}
    rotation {pitch=-30}
   }
   Save cached pose 'AnimDynCache'(Pose=Component To Local(ComponentPose=@n14)) {cacheName=AnimDynCache}
   n15 = Control Rig(Source=Use cached pose 'IniPoseCache'()) {
    alphaBoolBlend {blendOption=Linear}
    controlRigAssetReference {blueprintRigClass=&/M_MED_SlyJudge_CtrlRig.M_MED_SlyJudge_CtrlRig_C}
   }
   Save cached pose 'controlRigCache'(Pose=@n15) {cacheName=controlRigCache}
  ]}
 EventGraph {nodes [
   Event BlueprintUpdateAnimation()
   > n1 = Gravity Override State Function DEPRECATED(
    CharacterPartAnimInstance=Self-Reference()
    GameTimeInSeconds=GetGameTimeInSeconds()
    GravityOverrideParams=Get Gravity Override Params()
    RBAN Params=Get RBAN Params()
   )
   > Set GravityOverride(GravityOverride=@n1)
   > Set Sim Space Settings(Sim Space Settings:Sim Space Settings Structure=Break RBANSetting Struct!(RBANSettingStruct:"RBANSetting Struct Structure (by ref)"=@n1.RBAN SettingsOut) -> (
     ComponentLinearAccScale_28_#1:Vector=~
     ComponentLinearVelScale_31_#2:Vector=~
     SimSpaceSettings_34_#3:Sim Space Settings Structure
    ) {is=title K2Node_BreakStruct}.SimSpaceSettings_34_#3) -> (Output_Get:Sim Space Settings Structure)
   n2 = Get bIsTacticalSprinting()
   > Set isRegularJumping(isRegularJumping:bool=AND Boolean(A=Get bIsJumping(), B=NOT Boolean(A=@n2))) -> (Output_Get:bool)
   n3 = Get bRigidBodyEnabled()
   > Set RBAN OFF(RBAN OFF:bool=AND Boolean!(A=NOT Boolean(A=@n3), B=Get bIsCrouching(), C:bool=Get bIsMoving2D()) -> (ReturnValue)) -> (Output_Get:bool)
   > n4 = Branch(Condition=Get bIsInFrontEnd())
   > Set ChangeSimulationSpace(ChangeSimulationSpace:Byte=2) -> (Output_Get:Byte=0)
   Set ChangeSimulationSpace(execute=@n4.else, ChangeSimulationSpace=3) -> (Output_Get=0)
  ]}
]
variables [
 ChangeSimulationSpace {type=integer}
 crouching sprinting {type=boolean}
 Gravity Override Params {type=GravityOverrideParamsStruct, default {tactical_Sprint {gravityOverride=<0 0 -980>}}}
 GravityOverride {type=Vector}
 IsCrouching {type=boolean}
 isRegularJumping {type=boolean}
 Left {type=boolean}
 RBAN OFF {type=boolean}
 RBAN Params {
  type=RBANParamsStruct
  default {
   dBNO {simSpaceSettings {
     maxAngularAcceleration=1000
     maxAngularVelocity=3
     maxLinearAcceleration=300
     maxLinearVelocity=700
     velocityScaleZ=0
     externalLinearDragV=<0 0.5 0>
    }}
   dBNO_Carried {simSpaceSettings {
     maxAngularAcceleration=1000
     maxAngularVelocity=3
     maxLinearAcceleration=300
     maxLinearVelocity=700
     velocityScaleZ=0
     externalLinearDragV=<0 0.5 0>
    }}
   emoteOrMelee {simSpaceSettings {
     maxAngularAcceleration=1000
     maxAngularVelocity=3
     maxLinearAcceleration=300
     maxLinearVelocity=700
     velocityScaleZ=0
     externalLinearDragV=<0 0.5 0>
    }}
   front_End {simSpaceSettings {
     maxAngularAcceleration=1000
     maxAngularVelocity=3
     maxLinearAcceleration=150
     maxLinearVelocity=350
     velocityScaleZ=0
     worldAlpha=0.3
     externalLinearDragV=<0 0.5 0>
    }}
   onGround_Crouch_Moving {
    componentLinearAccScale=<0.3 0 0>
    componentLinearVelScale=<0.3 0 0>
    simSpaceSettings {
     maxAngularAcceleration=1000
     maxAngularVelocity=3
     maxLinearAcceleration=300
     maxLinearVelocity=700
     velocityScaleZ=0.1
     worldAlpha=0.1
     externalLinearDragV=<1 1 1>
    }
   }
   onGround_Falling {simSpaceSettings {
     maxAngularAcceleration=1000
     maxAngularVelocity=3
     maxLinearAcceleration=300
     maxLinearVelocity=700
     velocityScaleZ=0
     externalLinearDragV=<0 0.5 0>
    }}
   onGround_Moving {simSpaceSettings {
     maxAngularAcceleration=1000
     maxAngularVelocity=3
     maxLinearAcceleration=300
     maxLinearVelocity=700
     velocityScaleZ=0
     externalLinearDragV=<0 0.5 0>
    }}
   onGround_Standing {
    componentLinearAccScale=<0.3 0.3 0.3>
    componentLinearVelScale=<0.3 0.3 0.3>
    simSpaceSettings {
     maxAngularAcceleration=1000
     maxAngularVelocity=3
     maxLinearAcceleration=300
     maxLinearVelocity=700
     velocityScaleZ=0.5
     worldAlpha=0.8
     externalLinearDragV=<1 1 1>
    }
   }
   rail_Grinding {simSpaceSettings {
     maxAngularAcceleration=1000
     maxAngularVelocity=3
     maxLinearAcceleration=300
     maxLinearVelocity=700
     velocityScaleZ=0
     worldAlpha=0.1
     externalLinearDragV=<0 0.5 0>
    }}
   riding_Moving {simSpaceSettings {
     maxAngularAcceleration=1000
     maxAngularVelocity=3
     maxLinearAcceleration=300
     maxLinearVelocity=700
     velocityScaleZ=0
     externalLinearDragV=<0 0.5 0>
    }}
   riding_Moving_Targeting {simSpaceSettings {
     maxAngularAcceleration=1000
     maxAngularVelocity=3
     maxLinearAcceleration=300
     maxLinearVelocity=700
     velocityScaleZ=0
     externalLinearDragV=<0 0.5 0>
    }}
   riding_Standing {simSpaceSettings {
     maxAngularAcceleration=1000
     maxAngularVelocity=3
     maxLinearAcceleration=300
     maxLinearVelocity=700
     velocityScaleZ=0
     externalLinearDragV=<0 0.5 0>
    }}
   skydive_DivingDown {simSpaceSettings {
     maxAngularAcceleration=1000
     maxAngularVelocity=3
     maxLinearAcceleration=300
     maxLinearVelocity=700
     velocityScaleZ=0
     externalLinearDragV=<0 0.5 0>
    }}
   skydive_FloatingDown {
    componentLinearAccScale=<0 0.5 0>
    componentLinearVelScale=<0 0.5 0>
    simSpaceSettings {
     maxAngularAcceleration=1000
     maxAngularVelocity=3
     maxLinearAcceleration=300
     maxLinearVelocity=700
     velocityScaleZ=0
     externalLinearDragV=<0 0.5 0>
    }
   }
   skydive_GoingUp {simSpaceSettings {
     maxAngularAcceleration=1000
     maxAngularVelocity=3
     maxLinearAcceleration=300
     maxLinearVelocity=700
     velocityScaleZ=0
     externalLinearDragV=<0 0.5 0>
    }}
   skydive_Parachuting {
    componentLinearAccScale=<0 0.1 0.3>
    componentLinearVelScale=<0 0.1 0.3>
    simSpaceSettings {
     maxAngularAcceleration=1000
     maxAngularVelocity=3
     maxLinearAcceleration=300
     maxLinearVelocity=700
     velocityScaleZ=0
     externalLinearDragV=<0 0.5 0>
    }
   }
   skydive_Surfing {simSpaceSettings {
     maxAngularAcceleration=1000
     maxAngularVelocity=3
     maxLinearAcceleration=300
     maxLinearVelocity=700
     velocityScaleZ=0
     externalLinearDragV=<0 0.5 0>
    }}
   swimming {simSpaceSettings {
     maxAngularAcceleration=1000
     maxAngularVelocity=3
     maxLinearAcceleration=300
     maxLinearVelocity=700
     velocityScaleZ=0
     externalLinearDragV=<0 0.5 0>
    }}
   tactical_Sprint {
    componentLinearAccScale=<0.01 0 0>
    componentLinearVelScale=<0.01 0 0>
    simSpaceSettings {
     maxAngularAcceleration=1
     maxAngularVelocity=1
     maxLinearAcceleration=100
     maxLinearVelocity=10
     velocityScaleZ=0.1
     worldAlpha=0.1
     externalLinearDragV=<0.1 0.1 0.1>
    }
   }
   ziplining {simSpaceSettings {
     maxAngularAcceleration=1000
     maxAngularVelocity=3
     maxLinearAcceleration=300
     maxLinearVelocity=700
     velocityScaleZ=0
     externalLinearDragV=<0 0.5 0>
    }}
  }
 }
 Right {type=boolean}
 Sim Space Settings {
  type=SimSpaceSettings
  default {
   dampingAlpha=1
   maxAngularAcceleration=10000
   maxAngularVelocity=10000
   maxLinearAcceleration=10000
   maxLinearVelocity=10000
   velocityScaleZ=1
   worldAlpha=0
   externalLinearDragV=<0 0 0>
  }
 }
 SprintingRBAN {type=number}
]
