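# Control Rig for the SlyJudge character's face accessory mesh that procedurally corrects the hood, drawstrings and zipper as the head and body move.
# HoodFix offsets hood and zipper bones using remapped head-to-neck rotation and spherical pose readers, with one corrective smoothed by AlphaInterp.
# Multi-weighted parent constraints then mimic surface attachment for string, zipper and rivet bones, and aim constraints point the strings at their rivets.
# A Debug-gated step, off by default, drives head, neck and clavicles from test controls.
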ControlRigBlueprint /BRCosmetics/Characters/Player/Male/Medium/Bodies/M_MED_SlyJudge/Meshes/Parts/M_MED_SlyJudge_FaceAcc_CtrlRig
graphs [
 New Function_ContainedGraph {nodes [
   n1 = RigVMFunctionEntryNode()
   RigVMFunction_ControlFlowBranch = &RigVMFunction_ControlFlowBranch(Condition=@n1.Debug, ExecuteContext=@n1.ExecuteContext)
   RigVMFunctionReturnNode(ExecuteContext=@RigVMFunction_ControlFlowBranch.Completed)
   RigUnit_GetTransform = &RigUnit_GetTransform(Item=(Type=Control,Name="head_ctrl"), bInitial=False)
   RigUnit_SetTransform = &RigUnit_SetTransform(
    Item=(Type=Bone,Name="head")
    Weight=1
    bInitial=False
    Value=@RigUnit_GetTransform.Transform
    ExecutePin=@RigVMFunction_ControlFlowBranch.True
   )
   > &RigUnit_SetTransform(
    Item=(Type=Bone,Name="neck_02")
    Weight=1
    bInitial=False
    Value=&RigUnit_GetTransform(Item=(Type=Control,Name="neck_02_ctrl"), bInitial=False).Transform
   )
   > &RigUnit_SetTransform(
    Item=(Type=Bone,Name="clavicle_l")
    Weight=1
    bInitial=False
    Value=&RigUnit_GetTransform(Item=(Type=Control,Name="clavicle_l_ctrl"), bInitial=False).Transform
   )
   > &RigUnit_SetTransform(
    Item=(Type=Bone,Name="clavicle_r")
    Weight=1
    bInitial=False
    Value=&RigUnit_GetTransform(Item=(Type=Control,Name="clavicle_r_ctrl"), bInitial=False).Transform
   )
  ]}
 New Function_ContainedGraph {nodes [
   n1 = RigVMFunctionEntryNode()
   RigVMFunction_MathQuaternionToEuler = &RigVMFunction_MathQuaternionToEuler(
    RotationOrder=ZYX
    Value=&RigUnit_GetRelativeTransformForItem(
     Child=(Type=Bone,Name="head")
     Parent=(Type=Bone,Name="neck_01")
     bChildInitial=False
     bParentInitial=False
     Child.Name=@n1.Child
     Parent.Name=@n1.Parent
    ).RelativeTransform.Rotation
   )
   Remap = &RigVMFunction_MathDoubleRemap(
    TargetMaximum=1
    TargetMinimum=0
    SourceMaximum=@n1.Source Max
    SourceMinimum=@n1.Source Min
    Value=@RigVMFunction_MathQuaternionToEuler.Result.Z
   )
   > RigVMFunctionReturnNode(
    Delta_z=@Remap.Result
    Delta_X_Negative=&RigVMFunction_MathDoubleRemap(SourceMaximum=-60, SourceMinimum=0, TargetMaximum=1, TargetMinimum=0, Value=@RigVMFunction_MathQuaternionToEuler.Result.X).Result
    Delta_X_Positive=&RigVMFunction_MathDoubleRemap(SourceMaximum=70, SourceMinimum=0, TargetMaximum=1, TargetMinimum=0, Value=@RigVMFunction_MathQuaternionToEuler.Result.X).Result
   )
  ]}
 New Function_ContainedGraph {nodes [
   n1 = RigVMFunctionEntryNode()
   RigVMFunction_Sequence = &RigVMFunction_Sequence(ExecuteContext=@n1.ExecuteContext)
   DeltaRotation = RigVMFunctionReferenceNode(Child=Head, Parent=neck_01, Source Max=-60, Source Min=-10, ExecuteContext=@RigVMFunction_Sequence.A)
   > &RigUnit_ModifyTransforms(
    ItemToModify=((Item=(Type=Bone,Name="base_hood_l"),Transform=(Rotation=(X=-0.0545,Y=0.059,Z=-0.1738,W=0.9815),Translation=(X=0,Y=0,Z=0),Scale3D=(X=1,Y=1,Z=1))))
    WeightMinimum=0
    Weight=@DeltaRotation.Delta_X_Negative
   )
   > &RigUnit_ModifyTransforms(
    ItemToModify=((Item=(Type=Bone,Name="base_hood_r"),Transform=(Rotation=(X=-0.092,Y=0.0488,Z=0.2802,W=0.9543),Translation=(X=0,Y=0,Z=0),Scale3D=(X=1,Y=1,Z=1))))
    WeightMinimum=0
    Weight=@DeltaRotation.Delta_X_Positive
   )
   > &RigUnit_ModifyTransforms(
    ItemToModify=((Item=(Type=Bone,Name="dyn_hood_c"),Transform=(Rotation=(X=0,Y=0,Z=0,W=1),Translation=(X=30,Y=0,Z=0),Scale3D=(X=1,Y=1,Z=1))))
    WeightMinimum=0
    Weight=@DeltaRotation.Delta_z
   )
   > DeltaRotation_1 = RigVMFunctionReferenceNode(Child=Head, Parent=neck_02, Source Max=5, Source Min=-30)
   > &RigUnit_ModifyTransforms(
    ItemToModify=((Item=(Type=Bone,Name="dyn_hood_front"),Transform=(Rotation=(X=0,Y=0,Z=0,W=1),Translation=(X=0,Y=3,Z=0),Scale3D=(X=1,Y=1,Z=1))))
    WeightMinimum=0
    Weight=@DeltaRotation_1.Delta_z
   )
   > n2 = &RigUnit_SphericalPoseReader(
    ActiveRegionScaleFactors=(PositiveWidth=1,NegativeWidth=1,PositiveHeight=1,NegativeHeight=1)
    ActiveRegionSize=0.02
    Debug=(bDrawDebug=false,bDraw2D=False,bDrawLocalAxes=False,DebugScale=25,DebugSegments=20,DebugThickness=0.25)
    DriverAxis=(X=0,Y=1,Z=0)
    DriverItem=(Type=Bone,Name="head")
    FalloffRegionScaleFactors=(PositiveWidth=1,NegativeWidth=1,PositiveHeight=1,NegativeHeight=1)
    FalloffSize=0.05
    FlipHeightScaling=False
    FlipWidthScaling=False
    OptionalParentItem=(Type=Bone,Name="spine_05")
    RotationOffset=(X=115,Y=0,Z=30)
   )
   > &RigUnit_ModifyTransforms(
    ItemToModify=((Item=(Type=Bone,Name="dyn_zipper_base"),Transform=(Rotation=(X=0,Y=0,Z=0.2588,W=0.9659),Translation=(X=0,Y=0,Z=0),Scale3D=(X=1,Y=1,Z=1))))
    WeightMinimum=0
    Weight=@n2.OutputParam
   )
   > n3 = &RigUnit_SphericalPoseReader(
    ActiveRegionScaleFactors=(PositiveWidth=1,NegativeWidth=1,PositiveHeight=1,NegativeHeight=1)
    ActiveRegionSize=0.05
    Debug=(bDrawDebug=false,bDraw2D=False,bDrawLocalAxes=False,DebugScale=25,DebugSegments=20,DebugThickness=0.25)
    DriverAxis=(X=0,Y=1,Z=0)
    DriverItem=(Type=Bone,Name="head")
    FalloffRegionScaleFactors=(PositiveWidth=1,NegativeWidth=1,PositiveHeight=1,NegativeHeight=1)
    FalloffSize=0.1
    FlipHeightScaling=False
    FlipWidthScaling=False
    OptionalParentItem=(Type=Bone,Name="None")
    RotationOffset=(X=117,Y=0,Z=21)
   )
   Multiply = &RigVMFunction_MathDoubleMul(B=@DeltaRotation_1.Delta_z, A=@n3.OutputParam)
   AlphaInterp = &RigVMFunction_AlphaInterp(InterpSpeedDecreasing=12, InterpSpeedIncreasing=12, Scale=1, bClampResult=true, bMapRange=true, Value=@Multiply.Result)
   > &RigUnit_ModifyTransforms(
    ItemToModify="((Item=(Type=Bone,Name=\"base_hood_l\"),Transform=(Rotation=(X=-0.1318,Y=0.0491,Z=-0.0288,W=0.9896),Translation=(X=-0.5354,Y=0.1122,Z=0.3962),Scale3D=(X=1,Y=1,Z=1))),(Item=(Type=Bone,Name=\"base_hood_front\"),Transform=(Rotation=(X=-0.0658,Y=0.142,Z=0.1306,W=0 ...(truncated)"
    WeightMinimum=0
    Weight=@AlphaInterp.Result
   )
   RigVMFunctionReturnNode(ExecuteContext=@RigVMFunction_Sequence.B)
  ]}
 New Function_ContainedGraph {nodes [
   RigVMFunctionEntryNode()
   > &RigUnit_ParentConstraint(
    AdvancedSettings=(RotationOrderForFilter=XZY)
    Child=(Type=Bone,Name="dyn_string_l_base")
    Parents="((Item=(Type=Bone,Name=\"clavicle_r\"),Weight=0.0002),(Item=(Type=Bone,Name=\"clavicle_l\"),Weight=0.0203),(Item=(Type=Bone,Name=\"neck_01\"),Weight=0.0381),(Item=(Type=Bone,Name=\"dyn_hood_front\"),Weight=0.0925),(Item=(Type=Bone,Name=\"neck_02\"),Weight=0.1033),(I ...(truncated)"
    Weight=1
   )
   > &RigUnit_ParentConstraint(
    AdvancedSettings=(RotationOrderForFilter=XZY)
    Child=(Type=Bone,Name="dyn_string_r_base")
    Parents="((Item=(Type=Bone,Name=\"clavicle_l\"),Weight=0.0002),(Item=(Type=Bone,Name=\"clavicle_r\"),Weight=0.0192),(Item=(Type=Bone,Name=\"neck_01\"),Weight=0.0384),(Item=(Type=Bone,Name=\"neck_02\"),Weight=0.1046),(Item=(Type=Bone,Name=\"dyn_hood_front\"),Weight=0.1085),(I ...(truncated)"
    Weight=1
   )
   > &RigUnit_ParentConstraint(
    AdvancedSettings=(RotationOrderForFilter=XZY)
    Child=(Type=Bone,Name="base_main_zipper")
    Parents="((Item=(Type=Bone,Name=\"clavicle_l\"),Weight=0.001),(Item=(Type=Bone,Name=\"clavicle_r\"),Weight=0.001),(Item=(Type=Bone,Name=\"neck_01\"),Weight=0.006),(Item=(Type=Bone,Name=\"neck_02\"),Weight=0.0149),(Item=(Type=Bone,Name=\"spine_05\"),Weight=0.0792),(Item=(Type ...(truncated)"
    Weight=1
   )
   > &RigUnit_ParentConstraint(
    AdvancedSettings=(RotationOrderForFilter=XZY)
    Child=(Type=Bone,Name="base_rivet_jnt_string_l_aim")
    Parents="((Item=(Type=Bone,Name=\"deform_upperarm_fr_l\"),Weight=0.0013),(Item=(Type=Bone,Name=\"spine_01\"),Weight=0.0171),(Item=(Type=Bone,Name=\"clavicle_l\"),Weight=0.0551),(Item=(Type=Bone,Name=\"spine_02\"),Weight=0.1113),(Item=(Type=Bone,Name=\"deform_pec_l\"),Weight= ...(truncated)"
    Weight=1
   )
   > &RigUnit_ParentConstraint(
    AdvancedSettings=(RotationOrderForFilter=XZY)
    Child=(Type=Bone,Name="base_rivet_jnt_string_r_aim")
    Parents="((Item=(Type=Bone,Name=\"deform_upperarm_fr_r\"),Weight=0.0036),(Item=(Type=Bone,Name=\"spine_01\"),Weight=0.0074),(Item=(Type=Bone,Name=\"clavicle_r\"),Weight=0.0727),(Item=(Type=Bone,Name=\"spine_02\"),Weight=0.089),(Item=(Type=Bone,Name=\"deform_pec_r\"),Weight=0 ...(truncated)"
    Weight=1
   )
   > RigVMFunctionReturnNode()
  ]}
 New Function_ContainedGraph {nodes [
   RigVMFunctionEntryNode()
   > &RigUnit_AimConstraintLocalSpaceOffset(
    Child=(Type=Bone,Name="dyn_string_l_base")
    Filter=(bX=True,bY=True,bZ=True)
    Parents=((Item=(Type=Bone,Name="rivet_jnt_string_l_aim"),Weight=1))
    UpAxis=(X=0,Y=0,Z=-1)
    Weight=1
    WorldUp=(Target=(X=0,Y=0,Z=0),Kind=Direction,Space=(Type=None,Name="None"))
   )
   > &RigUnit_AimConstraintLocalSpaceOffset(
    Child=(Type=Bone,Name="dyn_string_r_base")
    Filter=(bX=True,bY=True,bZ=True)
    Parents=((Item=(Type=Bone,Name="rivet_jnt_string_r_aim"),Weight=1))
    UpAxis=(X=0,Y=0,Z=-1)
    Weight=1
    WorldUp=(Target=(X=0,Y=0,Z=0),Kind=Direction,Space=(Type=None,Name="None"))
   )
   > RigVMFunctionReturnNode()
  ]}
 RigVMFunctionLibrary {nodes [
   DeltaRotation = RigVMCollapseNode(ExecuteContext=())
   HoodFix = RigVMCollapseNode(ExecuteContext=())
   Strings = RigVMCollapseNode(ExecuteContext=())
   controls = RigVMCollapseNode(Debug=False, ExecuteContext=())
   surfaceConstraints = RigVMCollapseNode(ExecuteContext=())
  ]}
 RigVMModel {nodes [
   RigUnit_BeginExecution = &RigUnit_BeginExecution()
   controls = RigVMFunctionReferenceNode(Debug=false, ExecuteContext=@RigUnit_BeginExecution.ExecutePin)
   > HoodFix = RigVMFunctionReferenceNode()
   > surfaceConstraints = RigVMFunctionReferenceNode()
   > Strings = RigVMFunctionReferenceNode()
  ]}
]
hierarchy [
 root {loc=<0 0 0>, rot=<0 0 0 1>}
 pelvis {parent=root, loc=<0 -0.7071 93.1191>, rot=<0.0127 0.707 -0.0127 -0.707>}
 spine_01 {parent=pelvis, loc=<6.7494 -0.6388 0>, rot=<0 0 0.034 -0.9994>}
 spine_02 {parent=spine_01, loc=<7.2518 -0.035 0>, rot=<0 0 0.0238 -0.9997>}
 spine_03 {parent=spine_02, loc=<6.0104 0 0>, rot=<0 0 0.0149 -0.9999>}
 spine_04 {parent=spine_03, loc=<8.1024 0 0>, rot=<0 0 -0.0132 -0.9999>}
 spine_05 {parent=spine_04, loc=<10.0146 0 0>, rot=<0 0 -0.0015 -1>}
 neck_01 {parent=spine_05, loc=<11.3138 0.1529 0>, rot=<0 0 -0.2003 -0.9797>}
 neck_02 {parent=neck_01, loc=<4.5627 0 0>, rot=<0 0 0.0504 -0.9987>}
 head {parent=neck_02, loc=<5.3914 0 0>, rot=<0 0 0.115 -0.9934>}
 base_hood_top {parent=head, loc=<9.5014 12.3618 0.003>, rot=<-0.2098 -0.6753 0.6753 0.2098>}
 dyn_hood_top {parent=base_hood_top, loc=<0 0 0>, rot=<0 0 0 1>}
 base_hood_c {parent=neck_02, loc=<3.9048 -7.6424 0.0355>, rot=<0.3604 -0.581 0.441 -0.5815>}
 dyn_hood_c {parent=base_hood_c, loc=<0 0 0>, rot=<0 0 0 1>}
 base_hood_front {parent=neck_02, loc=<0 0 0>, rot=<0 0 0 1>}
 dyn_hood_front {parent=base_hood_front, loc=<0 0 0>, rot=<0 0 0 1>}
 base_hood_l {parent=neck_02, loc=<3.4037 -7.695 -1.3221>, rot=<-0.0784 -0.7027 0.0784 -0.7027>}
 dyn_hood_l {parent=base_hood_l, loc=<0 0 0>, rot=<0 0 0 1>}
 base_hood_r {parent=neck_02, loc=<3.3982 -7.6175 1.3179>, rot=<-0.0784 -0.7027 0.0784 -0.7027>}
 dyn_hood_r {parent=base_hood_r, loc=<0 0 0>, rot=<0 0 0 1>}
 clavicle_l {parent=spine_05, loc=<4.3575 -0.1466 -2.3223>, rot=<0.0465 -0.7289 -0.009 -0.6829>}
 upperarm_l {parent=clavicle_l, loc=<15.5111 0 0>, rot=<0.0226 -0.3234 -0.0088 -0.9459>}
 upperarm_twist_01_l {parent=upperarm_l, loc=<0.0041 0 0>, rot=<0 0 0 -1>}
 deform_upperarm_fr_base_l {parent=upperarm_twist_01_l, loc=<-0.3691 7.0506 -2.765>, rot=<0 0 0 1>}
 deform_upperarm_fr_l {parent=deform_upperarm_fr_base_l, loc=<0 0 0>, rot=<0 0 0 1>}
 deform_upperarm_up_base_l {parent=upperarm_twist_01_l, loc=<-3.2216 -1.1391 3.3311>, rot=<0 0 0 1>}
 deform_upperarm_up_l {parent=deform_upperarm_up_base_l, loc=<0 0 0>, rot=<0 0 0 1>}
 clavicle_r {parent=spine_05, loc=<4.3575 -0.1466 2.3223>, rot=<0.7289 0.0465 -0.6829 0.009>}
 upperarm_r {parent=clavicle_r, loc=<-15.5111 0 0>, rot=<0.0226 -0.3234 -0.0088 -0.9459>}
 upperarm_twist_01_r {parent=upperarm_r, loc=<-0.0041 0 0>, rot=<0 0 0 -1>}
 deform_upperarm_up_base_r {parent=upperarm_twist_01_r, loc=<3.2214 1.1391 -3.3308>, rot=<0 0 0 1>}
 deform_upperarm_up_r {parent=deform_upperarm_up_base_r, loc=<0 0 0>, rot=<0 0 0 -1>}
 deform_upperarm_fr_base_r {parent=upperarm_twist_01_r, loc=<0.3689 -7.0506 2.7652>, rot=<0 0 0 1>}
 deform_upperarm_fr_r {parent=deform_upperarm_fr_base_r, loc=<0 0 0>, rot=<0 0 0 1>}
 dyn_string_l_base {parent=spine_05, loc=<11.6466 7.9265 -4.5062>, rot=<-0.2603 -0.6574 0.661 0.2511>}
 dyn_string_l_parent {parent=dyn_string_l_base, loc=<0 0 0>, rot=<0 0 0 1>}
 dyn_string_l_01 {parent=dyn_string_l_parent, loc=<3.6936 0 0.3072>, rot=<0.0594 0 0.0637 -0.9962>}
 dyn_string_l_02 {parent=dyn_string_l_01, loc=<4.0039 0.2459 0.3925>, rot=<-0.0297 0 -0.0319 -0.999>}
 dyn_string_l_03 {parent=dyn_string_l_02, loc=<3.8946 -0.0343 0.5323>, rot=<0 0 0 1>}
 dyn_string_r_base {parent=spine_05, loc=<11.483 7.9894 4.4311>, rot=<-0.2003 -0.6782 0.6764 0.2062>}
 dyn_string_r_parent {parent=dyn_string_r_base, loc=<0 0 0>, rot=<0 0 0 1>}
 dyn_string_r_01 {parent=dyn_string_r_parent, loc=<3.5002 -0.0391 -0.2852>, rot=<0 0 0 1>}
 dyn_string_r_02 {parent=dyn_string_r_01, loc=<4.0132 0.2348 -0.2651>, rot=<0 0 0 1>}
 dyn_string_r_03 {parent=dyn_string_r_02, loc=<3.9283 0.1784 -0.113>, rot=<0 0 0 1>}
 base_main_zipper {parent=spine_05, loc=<10.5042 10.7196 0.0156>, rot=<0.2442 0.6636 0.6636 0.2442>}
 dyn_main_zipper {parent=base_main_zipper, loc=<0 0 0>, rot=<0 0 0 1>}
 dyn_zipper_base {parent=dyn_main_zipper, loc=<0.9651 -0.0552 -0.0492>, rot=<-0.6533 -0.2706 0.2706 -0.6533>}
 dyn_zipper_parent {parent=dyn_zipper_base, loc=<0 0 0>, rot=<0 0 0 1>}
 base_rivet_jnt_string_l_aim {parent=spine_05, loc=<-1.2177 19.6748 -4.7495>, rot=<-0.2603 -0.6574 0.661 0.2511>}
 rivet_jnt_string_l_aim {parent=base_rivet_jnt_string_l_aim, loc=<0 0 0>, rot=<0 0 0 1>}
 base_rivet_jnt_string_r_aim {parent=spine_05, loc=<-1.218 16.7382 4.5698>, rot=<-0.2003 -0.6782 0.6764 0.2062>}
 rivet_jnt_string_r_aim {parent=base_rivet_jnt_string_r_aim, loc=<0 0 0>, rot=<0 0 0 1>}
 deform_pec_aim_base_l {parent=spine_04, loc=<7.8732 12.6388 -2.4123>, rot=<0.0102 -0.4602 0.3979 -0.7936>}
 deform_pec_aim_l {parent=deform_pec_aim_base_l, loc=<0 0 0>, rot=<0 0 0 1>}
 deform_pec_base_l {parent=deform_pec_aim_l, loc=<10.766 0 0>, rot=<0 0 0 1>}
 deform_pec_l {parent=deform_pec_base_l, loc=<0.055 0 0>, rot=<0 0 0 1>}
 deform_pec_aim_base_r {parent=spine_04, loc=<7.8727 12.6388 2.4123>, rot=<-0.4602 -0.0102 0.7936 0.3979>}
 deform_pec_aim_r {parent=deform_pec_aim_base_r, loc=<0 0 0>, rot=<0 0 0 1>}
 deform_pec_base_r {parent=deform_pec_aim_r, loc=<-10.7662 -0.0006 -0.0007>, rot=<0 0 0 1>}
 deform_pec_r {parent=deform_pec_base_r, loc=<0 0 0>, rot=<0 0 0 1>}
 neck_02_ctrl {controlType=EULER_TRANSFORM, type=Control, loc=<0 0 0>, rot=<0 0 0 1>}
 head_ctrl {controlType=EULER_TRANSFORM, parent=neck_02_ctrl, type=Control, loc=<0 0 0>, rot=<0 0 0 1>}
 clavicle_l_ctrl {controlType=EULER_TRANSFORM, type=Control, loc=<0 0 0>, rot=<0 0 0 1>}
 clavicle_r_ctrl {controlType=EULER_TRANSFORM, type=Control, loc=<0 0 0>, rot=<0 0 0 1>}
]
