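# Physics asset for the face accessory of the SlyJudge medium male player body, covering two dangling string bone chains, left and right.
# Each chain runs from a small box base through a parent and three capsule segments, linked by constraints with matching swing limits and zero twist limit.
# Right-side capsules are much thicker than left.
# Two overlapping, unconstrained capsules on spine_05 form an upper-torso body, likely for the strings to collide against.
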PhysicsAsset /BRCosmetics/Characters/Player/Male/Medium/Bodies/M_MED_SlyJudge/Meshes/Parts/M_MED_SlyJudge_FaceAcc_Physics
bodies {
 dyn_string_l_01 [{shapeType=Capsule, length=3.7257, radius=0.5433, shapeName=dyn_string_l_01_capsule, center=<2 0 0>, rotation=<rot 83.5502 179.9999 179.9999>}]
 dyn_string_l_02 [{
   shapeType=Capsule
   length=3.4825
   radius=0.5655
   shapeName=dyn_string_l_02_capsule
   center=<1.9788 0 0.2458>
   rotation=<rot 81.9305 179.9969 179.9963>
  }]
 dyn_string_l_03 [{shapeType=Capsule, length=2.4284, radius=0.6401, shapeName=dyn_string_l_03_capsule, center=<1.2402 0 0.0953>, rotation=<rot 85.645 180 180>}]
 dyn_string_l_base [{shapeType=Box, extentX=0.5, extentY=0.5, extentZ=0.5}]
 dyn_string_l_parent [{shapeType=Capsule, length=3.3709, radius=0.5541, shapeName=dyn_string_l_parent_capsule, center=<2 0 0>, rotation=<rot 85.5891 -180 -180>}]
 dyn_string_r_01 [{
   shapeType=Capsule
   length=3.7761
   radius=1.5249
   shapeName=dyn_string_r_01_capsule
   center=<1.9722 0 0.0122>
   rotation=<rot 86.5361 0.0001 0.0001>
  }]
 dyn_string_r_02 [{
   shapeType=Capsule
   length=4.2501
   radius=1.5427
   shapeName=dyn_string_r_02_capsule
   center=<1.9973 0 0.2141>
   rotation=<rot 88.099 179.9818 179.9822>
  }]
 dyn_string_r_03 [{
   shapeType=Capsule
   length=4.0298
   radius=1.4505
   shapeName=dyn_string_r_03_capsule
   center=<2.0206 0 0.3446>
   rotation=<rot 85.5185 179.9999 180>
  }]
 dyn_string_r_base [{shapeType=Box, extentX=0.5, extentY=0.5, extentZ=0.5, rotation=<rot 0.0149 1.0908 0.3623>}]
 dyn_string_r_parent [{
   shapeType=Capsule
   length=2.9673
   radius=1.8775
   shapeName=dyn_string_r_parent_capsule
   center=<1.9808 0 -0.2497>
   rotation=<rot 85.6024 0 0>
  }]
 spine_05 [
  {shapeType=Capsule, length=19.7762, radius=10.4311, shapeName=spine_05_capsule, center=<0.5043 0.0025 0.1006>}
  {shapeType=Capsule, length=17.6181, radius=10.7238, shapeName=spine_05_capsule, center=<-3.0393 1.7243 0.3829>}
 ]
}
constraints [
 {bone1Name=dyn_string_l_parent, bone2Name=dyn_string_l_base, swing1LimitDegrees=20, swing2LimitDegrees=20, twistLimitDegrees=0}
 {bone1Name=dyn_string_l_01, bone2Name=dyn_string_l_parent, swing1LimitDegrees=20, swing2LimitDegrees=20, twistLimitDegrees=0}
 {bone1Name=dyn_string_l_02, bone2Name=dyn_string_l_01, swing1LimitDegrees=20, swing2LimitDegrees=20, twistLimitDegrees=0}
 {bone1Name=dyn_string_l_03, bone2Name=dyn_string_l_02, swing1LimitDegrees=20, swing2LimitDegrees=20, twistLimitDegrees=0}
 {bone1Name=dyn_string_r_parent, bone2Name=dyn_string_r_base, swing1LimitDegrees=20, swing2LimitDegrees=20, twistLimitDegrees=0}
 {bone1Name=dyn_string_r_01, bone2Name=dyn_string_r_parent, swing1LimitDegrees=20, swing2LimitDegrees=20, twistLimitDegrees=0}
 {bone1Name=dyn_string_r_02, bone2Name=dyn_string_r_01, swing1LimitDegrees=20, swing2LimitDegrees=20, twistLimitDegrees=0}
 {bone1Name=dyn_string_r_03, bone2Name=dyn_string_r_02, swing1LimitDegrees=20, swing2LimitDegrees=20, twistLimitDegrees=0}
]
